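# Physics asset for the CirrusVine backpack cosmetic, giving its dangling cables and antenna bodies for simulated secondary motion.
# Capsules cover the root, antenna and cable bones, with one sphere for the pack.
# Constraints chain three cables and an upper branch off cable 2, mixing angle limits with free joints.
# Cables attach to the pack by tight, free or locked joints, while antenna and root have no constraints.
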PhysicsAsset /BRCosmetics/Accessories/FORT_Backpacks/Backpack_M_MED_CirrusVine/Meshes/M_MED_CirrusVine_Pack_Physics
bodies {
 dyn_antenna_1 [{
   shapeType=Capsule
   length=17.75
   radius=0.5
   shapeName=dyn_antenna_1_capsule
   center=<8.1492 -0.0001 0>
   rotation=<rot 90 -166.7373 -166.7364>
  }]
 dyn_cable_1_1 [{shapeType=Capsule, length=1.75, radius=0.5, shapeName=dyn_cable_1_1_capsule, center=<0.9429 0 0>, rotation=<rot -50.175 97.7934 -90.9792>}]
 dyn_cable_1_2 [{
   shapeType=Capsule
   length=1.75
   radius=0.5
   shapeName=dyn_cable_1_2_capsule
   center=<1.2484 -0.0104 -0.038>
   rotation=<rot 79.3124 74.3774 81.2039>
  }]
 dyn_cable_1_3 [{
   shapeType=Capsule
   length=1
   radius=0.5
   shapeName=dyn_cable_1_3_capsule
   center=<0.4083 0.4971 0.1408>
   rotation=<rot -5.2351 -24.984 63.0471>
  }]
 dyn_cable_1_end [{shapeType=Capsule, length=0, radius=0.75, shapeName=dyn_cable_1_end_capsule, rotation=<rot 0 0 -110.0008>}]
 dyn_cable_2_1 [{
   shapeType=Capsule
   length=0.25
   radius=2.25
   shapeName=dyn_cable_2_1_capsule
   center=<-1.1322 -1.3045 -1.5949>
   rotation=<rot 76.8695 87.0864 34.7267>
  }]
 dyn_cable_2_2 [{
   shapeType=Capsule
   length=2
   radius=0.75
   shapeName=dyn_cable_2_2_capsule
   center=<-1.1071 -0.2888 0.5033>
   rotation=<rot 60 0.0024 -19.9996>
  }]
 dyn_cable_2_3 [{
   shapeType=Capsule
   length=1.25
   radius=1
   shapeName=dyn_cable_2_3_capsule
   center=<-0.5599 0.5342 -0.0239>
   rotation=<rot -83.8301 -173.6456 163.7076>
  }]
 dyn_cable_2_4 [{shapeType=Capsule, length=1, radius=0.75, shapeName=dyn_cable_2_4_capsule, rotation=<rot 60 -139.9975 170.0002>}]
 dyn_cable_3_1 [{
   shapeType=Capsule
   length=0.75
   radius=0.5
   shapeName=dyn_cable_3_1_capsule
   center=<1.0154 0.2044 -0.029>
   rotation=<rot -90 8.8807 -3.8807>
  }]
 dyn_cable_3_2 [{
   shapeType=Capsule
   length=2.75
   radius=0.75
   shapeName=dyn_cable_3_2_capsule
   center=<1.6731 -0.0304 -0.1855>
   rotation=<rot 87.0203 85.6006 85.0184>
  }]
 dyn_cable_3_3 [{
   shapeType=Capsule
   length=1
   radius=0.75
   shapeName=dyn_cable_3_3_capsule
   center=<1.0514 -0.1646 0.2151>
   rotation=<rot -82.1359 -46.7094 44.66>
  }]
 dyn_cable_3_end [{
   shapeType=Capsule
   length=0
   radius=2.25
   shapeName=dyn_cable_3_end_capsule
   center=<-1.6946 -0.677 -1.2617>
   rotation=<rot -12.913 -26.8528 52.4792>
  }]
 dyn_cable_up_2_1 [{
   shapeType=Capsule
   length=2
   radius=0.5
   shapeName=dyn_cable_up_2_1_capsule
   center=<1.3575 0.151 -0.0035>
   rotation=<rot -84.497 152.9823 22.5762>
  }]
 dyn_cable_up_2_2 [{
   shapeType=Capsule
   length=0.75
   radius=0.5
   shapeName=dyn_cable_up_2_2_capsule
   center=<-0.0737 -0.3399 -0.006>
   rotation=<rot 8.5381 -12.3797 79.0723>
  }]
 dyn_cable_up_2_end [{
   shapeType=Capsule
   length=0.25
   radius=0.5
   shapeName=dyn_cable_up_2_end_capsule
   center=<-0.2731 -0.3255 -0.1139>
   rotation=<rot -9.5766 -79.0243 58.4003>
  }]
 pack [{shapeType=Sphere, radius=9, shapeName=pack_sphere, center=<0 -8.2535 0.5882>}]
 root [{shapeType=Capsule, length=25.25, radius=4.75, shapeName=root_capsule, center=<0 5.2912 0>}]
}
constraints [
 {bone1Name=dyn_cable_2_2, bone2Name=dyn_cable_2_1, swing1LimitDegrees=25, swing2LimitDegrees=25, twistLimitDegrees=15}
 {bone1Name=dyn_cable_2_3, bone2Name=dyn_cable_2_2, swing1LimitDegrees=35, swing2LimitDegrees=35, twistLimitDegrees=5}
 {bone1Name=dyn_cable_2_3, bone2Name=dyn_cable_2_4, swing1LimitDegrees=35, swing2LimitDegrees=35, twistLimitDegrees=5}
 {
  bone1Name=dyn_cable_2_4
  bone2Name=pack
  swing1LimitDegrees=45
  swing1Motion=Free
  swing2LimitDegrees=45
  swing2Motion=Free
  twistLimitDegrees=45
  twistMotion=Free
 }
 {
  bone1Name=dyn_cable_up_2_1
  bone2Name=dyn_cable_2_1
  swing1LimitDegrees=25
  swing1Motion=Free
  swing2LimitDegrees=25
  swing2Motion=Free
  twistLimitDegrees=15
  twistMotion=Free
 }
 {bone1Name=dyn_cable_up_2_2, bone2Name=dyn_cable_up_2_end, swing1LimitDegrees=25, swing2LimitDegrees=25, twistLimitDegrees=15}
 {bone1Name=dyn_cable_up_2_1, bone2Name=dyn_cable_up_2_2, swing1LimitDegrees=25, swing2LimitDegrees=25, twistLimitDegrees=7}
 {bone1Name=dyn_cable_3_2, bone2Name=dyn_cable_3_1, swing1LimitDegrees=35, swing2LimitDegrees=35, twistLimitDegrees=25}
 {
  bone1Name=dyn_cable_1_3
  bone2Name=dyn_cable_1_end
  swing1LimitDegrees=25
  swing1Motion=Free
  swing2LimitDegrees=25
  swing2Motion=Free
  twistLimitDegrees=15
  twistMotion=Free
 }
 {bone1Name=dyn_cable_1_2, bone2Name=dyn_cable_1_1, swing1LimitDegrees=35, swing2LimitDegrees=35, twistLimitDegrees=5}
 {bone1Name=dyn_cable_1_3, bone2Name=dyn_cable_1_2, swing1LimitDegrees=25, swing2LimitDegrees=25, twistLimitDegrees=5}
 {bone1Name=pack, bone2Name=dyn_cable_1_1, swing1LimitDegrees=0.1, swing2LimitDegrees=3, twistLimitDegrees=1}
 {
  bone1Name=dyn_cable_3_3
  bone2Name=dyn_cable_3_end
  swing1LimitDegrees=25
  swing1Motion=Free
  swing2LimitDegrees=25
  swing2Motion=Free
  twistLimitDegrees=15
  twistMotion=Free
 }
 {bone1Name=dyn_cable_3_2, bone2Name=dyn_cable_3_3, swing1LimitDegrees=35, swing2LimitDegrees=35, twistLimitDegrees=25}
 {
  bone1Name=dyn_cable_3_end
  bone2Name=pack
  swing1LimitDegrees=45
  swing1Motion=Locked
  swing2LimitDegrees=45
  swing2Motion=Locked
  twistLimitDegrees=45
  twistMotion=Locked
 }
]
